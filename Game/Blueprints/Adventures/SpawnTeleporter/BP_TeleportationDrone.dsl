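; Teleportation drone actor that plays a spawn-in or spawn-out montage as a visual for teleporting.
; BeginPlay levels pitch and roll, then PlaySpawnAnim picks the montage from TeleportIn and InLobby, sets its play rate and binds montage end.
; Tick polls the mesh's dynamic material Opacity and hides the actor at zero.
; The montage-end callback disables ticking, and EndPlay unbinds it.
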(event EventBeginPlay
  (bind (_returnvalue_roll _returnvalue_pitch _returnvalue_yaw) (Math|Rotator|BreakRotator (Transformation|GetActorRotation)))
  ;[0 1 0.1073] Register for anim montage ending events and play the spawn in effects
  (Transformation|SetActorRotation 0 (Math|Rotator|MakeRotator 0.0 0.0 _returnvalue_yaw) false) ; Cancel pitch/roll
  (CallFunction|PlaySpawnAnim))

(event EventEndPlay
  (params (EndPlayReason :type EEndPlayReason%20Enum :default "Destroyed"))
  (bind _skeletalmeshcomponentbudgeted (Variables|Default|GetSkeletalMeshComponentBudgeted))
  (bind _returnvalue (Components|SkeletalMesh|GetAnimInstance _skeletalmeshcomponentbudgeted))
  ;[1 0.5341 0] EndPlay Unbind listeners
  (|CallParentFunction EndPlayReason)
  (Utilities|IsValid _returnvalue
    (:Is%20Valid
      (EventDispatchers|UnbindEventfromOnMontageEnded _returnvalue (EventDispatchers|CreateEvent :!event Custom|OnSpawnOutAnimEnded)))))

(event EventTick
  (params (DeltaSeconds :type Float%20%28single-precision%29))
  (bind _skeletalmeshcomponentbudgeted (Variables|Default|GetSkeletalMeshComponentBudgeted))
  (bind _returnvalue (Rendering|Material|GetMaterial _skeletalmeshcomponentbudgeted 0))
  (bind _asmaterial_instance_dynamic (Utilities|Casting|CastToMaterialInstanceDynamic _returnvalue))
  ;[0 1 0.9747] Tick: When the material loses all opacity hide it and turn off ticking.
  (|CallParentFunction DeltaSeconds)
  (Utilities|IsValid _asmaterial_instance_dynamic
    (:Is%20Valid
      (bind _returnvalue_2 (Rendering|Material|GetScalarParameterValue _asmaterial_instance_dynamic "Opacity" :!class /Script/Engine.MaterialInstanceDynamic))
      (if (<= _returnvalue_2 0.0)
        (Rendering|SetActorHiddenInGame true)
        (Actor|Tick|SetActorTickEnabled false)))))

(event Custom|OnSpawnOutAnimEnded
  (params (Montage :type Anim%20Montage%20Object%20Reference) (bInterrupted :type Boolean))
  ;[1 1 1] The drone spawn out animation has ended
  (Actor|Tick|SetActorTickEnabled false))

(fn GetSpawnAnimation
  (return (select (Variables|Default|GetTeleportIn) "/Game/Animation/Game/MainPlayer/Montages/Gadgets/Bot_Turret_Spawn_In_Montage.Bot_Turret_Spawn_In_Montage" (select (Variables|Default|GetInLobby) "/Game/Animation/Game/MainPlayer/Montages/Gadgets/Bot_Turret_Spawn_Out_Montage_Athena_Lobby.Bot_Turret_Spawn_Out_Montage_Athena_Lobby" "/Game/Animation/Game/MainPlayer/Montages/Gadgets/Bot_Turret_Spawn_Out_Montage_Athena.Bot_Turret_Spawn_Out_Montage_Athena"))))

(fn PlaySpawnAnim
  (bind _skeletalmeshcomponentbudgeted (Variables|Default|GetSkeletalMeshComponentBudgeted))
  (Development|SetHiddeninGame _skeletalmeshcomponentbudgeted false false)
  (Components|Animation|PlayAnimation _skeletalmeshcomponentbudgeted (CallFunction|GetSpawnAnimation) false)
  (Components|Animation|SetPlayRate _skeletalmeshcomponentbudgeted (Variables|Default|GetAnimPlayRate))
  (EventDispatchers|BindEventtoOnMontageEnded (Components|SkeletalMesh|GetAnimInstance _skeletalmeshcomponentbudgeted) (EventDispatchers|CreateEvent :!event OnSpawnOutAnimEnded)))

(fn ConstructionScript)
