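; Victory drone actor that plays teleport-in or teleport-out animation and effects around a target player.
; PlaySpawnAnim chooses the montage and attached or world-placed FX from teleport direction, Athena, lobby, decoy and NPC flags.
; NPCs and decoys flagged isNPC drop the drone mesh and tick, keeping effects and optional death audio.
; Otherwise Tick mirrors skeletal material opacity onto the static mesh until montage end hides the drone; an authority-only notify broadcasts completion.
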(event EventBeginPlay
  (bind (_returnvalue_roll _returnvalue_pitch _returnvalue_yaw) (Math|Rotator|BreakRotator (Transformation|GetActorRotation)))
  (bind _characterattached (Variables|Default|GetCharacterAttached))
  (bind _characterattached_2 (Variables|Default|GetCharacterAttached))
  (if (Variables|Default|GetisNPC)
    ;[1 1 1] NPC & Decoys  -  Removes the Drone element and only spawns Teleport Effects
    (Components|DestroyComponent (Variables|Default|GetSkeletalMeshComponentBudgeted))
    (Actor|Tick|SetActorTickEnabled false)
    (CallFunction|PlaySpawnAnim)
    (CallFunction|AddedDeathAudio (select (Variables|Default|GetisDecoy) 0 (Variables|Default|GetSoundOnNPCDeath)))
    (Utilities|FlowControl|Delay 0.2)
    (Components|Activation|Activate _characterattached false)
    (Utilities|FlowControl|Delay (select (Variables|Default|GetisDecoy) 2.0 1.5))
    (Effects|Components|ParticleSystem|ReleasetoPool _characterattached)
    (else
      ;[1 1 1] Register for anim montage ending events and play the spawn in effects
      (Transformation|SetActorRotation 0 (Math|Rotator|MakeRotator 0.0 0.0 _returnvalue_yaw) false) ; Cancel pitch/roll
      (CallFunction|InitDrone)
      (CallFunction|PlaySpawnAnim)
      (Utilities|FlowControl|Delay 0.2)
      (Components|Activation|Activate _characterattached_2 false)
      (Effects|Components|ParticleSystem|SetVectorParameter _characterattached_2 "PointPosition" (Transformation|GetWorldLocation (Variables|Default|GetSkeletalMeshComponentBudgeted)))
      (Utilities|FlowControl|Delay 1.5)
      (Effects|Components|ParticleSystem|ReleasetoPool _characterattached_2))))

(event EventTick
  (params (DeltaSeconds :type Float%20%28single-precision%29))
  (bind _skeletalmeshcomponentbudgeted (Variables|Default|GetSkeletalMeshComponentBudgeted))
  (bind _returnvalue (Rendering|Material|GetMaterial _skeletalmeshcomponentbudgeted 0))
  (bind _asmaterial_instance_dynamic (Utilities|Casting|CastToMaterialInstanceDynamic _returnvalue))
  ;[1 1 1] Transfer the skeletal mesh material animation to the static mesh
  (if (not (Variables|Default|GetisNPC))
    (Utilities|IsValid _asmaterial_instance_dynamic
      (:Is%20Valid
        (bind _returnvalue_2 (Rendering|Material|GetScalarParameterValue _asmaterial_instance_dynamic "Opacity" :!class /Script/Engine.MaterialInstanceDynamic))
        (Rendering|Material|SetScalarParameterValue (Variables|Default|GetStaticMeshMID) "Opacity" _returnvalue_2 :!class /Script/Engine.MaterialInstanceDynamic)))))

(event Custom|OnSpawnOutAnimEnded
  (params (Montage :type Anim%20Montage%20Object%20Reference) (bInterrupted :type Boolean))
  ;[1 1 1] The drone spawn out animation has ended
  (Development|SetHiddeninGame (Variables|Default|GetSkeletalMeshComponentBudgeted) true false)
  (Actor|Tick|SetActorTickEnabled false))

(fn AddedDeathAudio
  (params (Sound%20to%20Spawn :type Sound%20Base%20Object%20Reference))
  (Utilities|IsValid Sound%20to%20Spawn
    (:Is%20Valid
      (Audio|SpawnSoundatLocation Sound%20to%20Spawn (Transformation|GetActorLocation (Variables|Default|GetTargetPlayer))))))

(fn DecoyFXParameters
  (return (Utilities|Array|MakeArray (Utilities|Struct|MakeParticleSysParam "Lifetime" "PSPT_ScalarRand" 1.6 1.2) (Utilities|Struct|MakeParticleSysParam "Lifetime2" "PSPT_ScalarRand" 0.5 2.0) (Utilities|Struct|MakeParticleSysParam "Lifetime3" "PSPT_ScalarRand" 1.0 0.8) (Utilities|Struct|MakeParticleSysParam "Spark_LOW" "PSPT_Scalar" 3e+01) (Utilities|Struct|MakeParticleSysParam "Spark" "PSPT_Scalar" 1e+02) (Utilities|Struct|MakeParticleSysParam "Frag_LOW" "PSPT_Scalar" 1e+02) (Utilities|Struct|MakeParticleSysParam "Frag" "PSPT_Scalar" 2e+02) (Utilities|Struct|MakeParticleSysParam "Shimmer" "PSPT_Scalar" 4e+01) (Utilities|Struct|MakeParticleSysParam "BoneSocketActor" "PSPT_Actor" (Variables|Default|GetTargetPlayer)))))

(fn InitDrone)

(fn NotifyTeleportFinishedTriggered
  (Networking|SwitchHasAuthority
    (:Authority
      (Default|CallSpawnOutAnimEndedDispatcher self))))

(fn NPCFXParameters
  (return (Utilities|Array|MakeArray (Utilities|Struct|MakeParticleSysParam "Lifetime" "PSPT_ScalarRand" 0.8 0.6) (Utilities|Struct|MakeParticleSysParam "Lifetime2" "PSPT_ScalarRand" 0.25 1.0) (Utilities|Struct|MakeParticleSysParam "Lifetime3" "PSPT_ScalarRand" 0.5 0.4) (Utilities|Struct|MakeParticleSysParam "Spark_LOW" "PSPT_Scalar" 15.0) (Utilities|Struct|MakeParticleSysParam "Spark" "PSPT_Scalar" 5e+01) (Utilities|Struct|MakeParticleSysParam "Frag_LOW" "PSPT_Scalar" 5e+01) (Utilities|Struct|MakeParticleSysParam "Frag" "PSPT_Scalar" 1e+02) (Utilities|Struct|MakeParticleSysParam "Shimmer" "PSPT_Scalar" 2e+01) (Utilities|Struct|MakeParticleSysParam "BoneSocketActor" "PSPT_Actor" (Variables|Default|GetTargetPlayer)))))

(fn PlaySpawnAnim
  (bind _skeletalmeshcomponentbudgeted (Variables|Default|GetSkeletalMeshComponentBudgeted))
  (if (not (Variables|Default|GetisNPC))
    (Development|SetHiddeninGame _skeletalmeshcomponentbudgeted false false)
    (Components|Animation|PlayAnimation _skeletalmeshcomponentbudgeted (select (Variables|Default|GetTeleportIn) (select (Variables|Default|GetIsAthena) (select (Variables|Default|GetInLobby) "/Game/Animation/Game/MainPlayer/Montages/Gadgets/Bot_Turret_Spawn_Out_Montage_Athena_Lobby.Bot_Turret_Spawn_Out_Montage_Athena_Lobby" "/Game/Animation/Game/MainPlayer/Montages/Gadgets/Bot_Turret_Spawn_Out_Montage_Athena.Bot_Turret_Spawn_Out_Montage_Athena") "/Game/Animation/Game/MainPlayer/Montages/Gadgets/Bot_Turret_Spawn_In_Montage.Bot_Turret_Spawn_In_Montage") (select (Variables|Default|GetIsAthena) "/Game/Animation/Game/MainPlayer/Montages/Gadgets/Bot_Turret_Spawn_Out_Montage_Athena.Bot_Turret_Spawn_Out_Montage_Athena" "/Game/Animation/Game/MainPlayer/Montages/Gadgets/Bot_Turret_Spawn_Out_Montage.Bot_Turret_Spawn_Out_Montage")) false)
    (Components|Animation|SetPlayRate _skeletalmeshcomponentbudgeted (Variables|Default|GetAnimPlayRate))
    (EventDispatchers|BindEventtoOnMontageEnded (Components|SkeletalMesh|GetAnimInstance _skeletalmeshcomponentbudgeted) (EventDispatchers|CreateEvent :!event OnSpawnOutAnimEnded)))
  (bind _targetplayer (Variables|Default|GetTargetPlayer)
    (:then
      (Variables|Default|SetMesh%20for%20Attachment (Variables|/Script/Engine.Character|GetMesh _targetplayer))
      (if (Variables|Default|GetTeleportIn)
        (bind _returnvalue (Effects|SpawnEmitterAttached :EmitterTemplate "/Game/Effects/Fort_Effects/Effects/Characters/Generic/ResIn/P_Player_Teleport_In_Athena.P_Player_Teleport_In_Athena" :AttachToComponent (Variables|Default|GetMesh%20for%20Attachment) :AttachPointName "root" :Scale "1.000000,1.000000,1.000000" :LocationType "SnapToTarget" :bAutoDestroy false :PoolingMethod "ManualRelease" :bAutoActivate false)) ; TELEPORT IN
        (Variables|Default|SetCharacterAttached _returnvalue)
        (elif (Variables|Default|GetInLobby)
          (bind _returnvalue_2 (Effects|SpawnFXSystemAttached "/Game/Effects/Fort_Effects/Effects/Characters/Generic/ResIn/NS_Player_Teleport_Out_Athena.NS_Player_Teleport_Out_Athena" (Variables|Default|GetMesh%20for%20Attachment) "fx_Chest"))
          (Variables|Default|SetCharacterAttached _returnvalue_2)
          (elif (Variables|Default|GetisDecoy)
            (bind _returnvalue_3 (Effects|Components|ParticleSystem|SpawnFXAtLocation :FXSystemAsset "/Game/Effects/Fort_Effects/Effects/Characters/Generic/ResIn/NS_Decoy_Teleport_Out_Athena.NS_Decoy_Teleport_Out_Athena" :Location (Transformation|GetSocketLocation (Variables|Default|GetMesh%20for%20Attachment) "fx_Chest") :Scale "1.000000,1.000000,1.000000" :bAutoDestroy true :PoolingMethod "None" :bAutoActivate false))
            (Effects|Components|ParticleSystem|SetActorParameter _returnvalue_3 "BoneSocketActor" (Variables|Default|GetTargetPlayer))
            (Variables|Default|SetCharacterAttached _returnvalue_3)
            (elif (Variables|Default|GetisNPC)
              (bind _returnvalue_4 (Effects|Components|ParticleSystem|SpawnFXAtLocation "/Game/Effects/Fort_Effects/Effects/Characters/Generic/ResIn/NS_Player_Teleport_Out_Athena.NS_Player_Teleport_Out_Athena" (Transformation|GetSocketLocation (Variables|Default|GetMesh%20for%20Attachment) "fx_Chest")))
              (Effects|Components|ParticleSystem|SetActorParameter _returnvalue_4 "BoneSocketActor" (Variables|Default|GetTargetPlayer))
              (Variables|Default|SetCharacterAttached 0)
              (else
                (bind _returnvalue_5 (Effects|SpawnFXSystemAttached "/Game/Effects/Fort_Effects/Effects/Characters/Generic/ResIn/NS_Player_Teleport_Out_Athena.NS_Player_Teleport_Out_Athena" (Variables|Default|GetMesh%20for%20Attachment) "fx_Chest"))
                (Variables|Default|SetCharacterAttached _returnvalue_5)))))))
    (:else)))

(fn ConstructionScript)
