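# Masked, two-sided subsurface material for the spawn teleporter flag, with a banner icon over a rotatable two-color gradient.
# Vertex green and alpha drive darkened and brightened gradient bands, and inverted green limits the icon mask that also blends color, metallic and roughness values.
# Panning swirl-noise normals build a displacement vector weighted by cubed vertex red, plus layered surface normals blended with a shared Cloth function's normal.
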Material /Game/Blueprints/Adventures/SpawnTeleporter/M_SpawnFlag
graph {
 properties {blendMode=BLEND_Masked, pixelDepthOffsetMode=PDOM_Legacy, shadingModel=MSM_Subsurface, twoSided=true, usedShadingModels=Subsurface}
 nodes [
  n1 = TextureCoordinate()
  n2 = ScalarParameter() {group=Linear Gradient Settings, parameterName=ColorRotation}
  n3 = Subtract(A=ComponentMask(None=Rotator(Coordinate=@n1, Time=@n2)) {g=true}, B=ScalarParameter() {defaultValue=-0.3, parameterName=ColorOffset})
  n4 = ScalarParameter() {group=Linear Gradient Settings, parameterName=RGBAMin}
  n5 = ScalarParameter() {defaultValue=1, group=Linear Gradient Settings, parameterName=RGBAMax}
  n6 = LinearInterpolate(
   A=VectorParameter() {parameterName=RGBA0, defaultValue=<rgba 1 0.2719 0 1>}.RGB
   B=VectorParameter() {parameterName=RGBA1, defaultValue=<rgba 0.28 0.0595 0 1>}.RGB
   Alpha=Clamp(None=MaterialFunctionCall(Alpha=@n3, Min=@n4, Max=@n5) {materialFunction=&SmoothStep}.Result)
  )
  Add(A=@n6) {constB=0.5}
  n7 = VertexColor()
  Ceil(None=Add(A=@n7.G) {constB=0.1})
  Constant3Vector() {constant=<rgba 1 0.1208 0.9073 0>}
  Constant() {r=1}
  n8 = TextureCoordinate() {vTiling=1.5}
  n9 = ScalarParameter() {defaultValue=0.1, parameterName=Scale}
  n10 = Time()
  n11 = ScalarParameter() {defaultValue=2, parameterName=Speed}
  n12 = TextureSample(UVs=Panner(Coordinate=Multiply(A=@n8, B=@n9), Time=Multiply(A=@n10, B=@n11)) {speedX=0.1, speedY=0.025}) {samplerType=SAMPLERTYPE_Normal, texture=/Game/Global/Textures/Default/Noise/T_SwirlNoise_N}
  n13 = Constant3Vector() {constant=<rgba -1 -1 1 0>}
  LinearInterpolate(
   A=Constant3Vector() {constant=<rgba 0 0 1 0>}
   B=Multiply(A=Multiply(A=@n12.RGB, B=@n13), B=VectorParameter() {parameterName=WPOPower, defaultValue=<rgba 12 1 5 1>}.RGB)
   Alpha=Power(Base=@n7.R, Exp=Constant() {r=3})
  )
  n14 = VectorParameter() {parameterName=Barnner Icon Translation, defaultValue=<rgba 1 1 0 0>}
  n15 = Add(A=Multiply(A=TextureCoordinate(), B=AppendVector(A=@n14.R, B=@n14.R)), B=AppendVector(A=@n14.B, B=@n14.B))
  n16 = TextureSampleParameter2D(UVs=@n15) {parameterName=BannerIcon, texture=/Game/UI/Foundation/Textures/Banner/Event_Special/T-Banners-Icons-Fortnite-L}
  n17 = Multiply(A=@n16.R, B=OneMinus(None=VertexColor().G))
  n18 = VectorParameter() {parameterName=Normal Translation, defaultValue=<rgba 0.9609 0.7791 0 0.1101>}
  n19 = Add(A=Multiply(A=TextureCoordinate(), B=AppendVector(A=@n18.R, B=@n18.R)), B=AppendVector(A=@n18.B, B=@n18.B))
  LinearInterpolate(
   A=ScalarParameter() {defaultValue=0.35, parameterName=SpecRoughnessMin}
   B=ScalarParameter() {defaultValue=0.5, parameterName=SpecRoughnessMax}
   Alpha=Multiply(
    A=@n17
    B=TextureSampleParameter2D(UVs=@n19) {parameterName=Normal, samplerType=SAMPLERTYPE_Normal, texture=/Game/Environments/Sets/Junkyard/Textures/T_Tarp_N}.B
   )
  )
  n20 = Multiply(A=@n6) {constB=0.1}
  n21 = Frac(None=ComponentMask(None=TextureCoordinate()) {g=true})
  n22 = VertexColor()
  n23 = Multiply(
   A=MaterialFunctionCall(Alpha=OneMinus(None=@n21), Min=Constant(), Max=Constant() {r=0.6}) {materialFunction=&SmoothStep}.Result
   B=@n22.G
  )
  n24 = Multiply(A=@n6) {constB=3}
  n25 = Constant()
  n26 = Constant() {r=0.2}
  n27 = VectorParameter() {parameterName=BannerIconColor, defaultValue=<rgba 1 1 1 1>}
  LinearInterpolate(
   A=LinearInterpolate(
    A=LinearInterpolate(A=@n6, B=@n20, Alpha=@n23)
    B=Clamp(None=@n24)
    Alpha=MaterialFunctionCall(Alpha=@n22.A, Min=@n25, Max=@n26) {materialFunction=&SmoothStep}.Result
   )
   B=@n27.RGB
   Alpha=@n17
  )
  LinearInterpolate(A=ScalarParameter() {defaultValue=0.12, parameterName=Metallic Fabric}, B=ScalarParameter() {parameterName=Metallic Icon}, Alpha=@n17)
  n28 = Constant3Vector() {constant=<rgba 0 0 1 0>}
  n29 = TextureCoordinate() {uTiling=0.5}
  n30 = ScalarParameter() {defaultValue=1, parameterName=ScaleB}
  n31 = ScalarParameter() {defaultValue=1.25, parameterName=SpeedB}
  n32 = TextureSample(UVs=Panner(Coordinate=Multiply(A=@n29, B=@n30), Time=Multiply(A=@n10, B=@n31)) {speedX=0.1, speedY=-0.05}) {samplerType=SAMPLERTYPE_Normal, texture=&T_SwirlNoise_N}
  n33 = VectorParameter() {parameterName=SmallNormalsPower, defaultValue=<rgba 0.2 0.2 1 1>}
  n34 = Multiply(
   A=MaterialFunctionCall(BaseNormal=@n12.RGB, AdditionalNormal=Multiply(A=@n32.RGB, B=@n33.RGB) {constB=-1}) {materialFunction=&BlendAngleCorrectedNormals}.Result
   B=Constant3Vector() {constant=<rgba -1 -1 1 0>}
  )
  MaterialFunctionCall(
   BaseNormal=Multiply(
    A=VectorParameter() {parameterName=MasterNormalPower, defaultValue=<rgba 0.5 0.5 0.5 1>}.RGB
    B=LinearInterpolate(A=@n28, B=@n34, Alpha=@n7.R)
   )
   AdditionalNormal=MaterialFunctionCall!(
    UVs
    Pan Speed=ScalarParameter() {defaultValue=0.469, group=WPO_ClothAnimation, parameterName=WPO_ClothPanSpeed}
    WPO_ClothOffsetDirection
    WPO_ClothNormalIntensity=Constant() {r=-0.1}
    WPO_ClothOffsetIntensity=Constant()
    "Motion Mask (Defaults to the Red Vertex Color)"
    WPO_ClothWPOWindTextureOffset=Constant() {r=2}
    Pan Direction=Constant() {r=13}
   ) {materialFunction=/Game/Packages/Fortress_SharedMaterials/Base_Material_Functions/Cloth}.Normal
  ) {materialFunction=&BlendAngleCorrectedNormals}
  Multiply(A=@n16.RGB, B=@n27.RGB)
  ScalarParameter() {defaultValue=0.12, parameterName=Specular}
  TextureSample(UVs=@n15) {texture=/Game/Environments/Hexmap/Outpost/T-Banners-Icons-Mask}
 ]
}
