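# Surface material for the pole of the spawn flag in the Adventures spawn teleporter, using Default Lit shading.
# It holds only flat values: a grey color constant, a grey vector parameter and three scalar constants, with no textures or math nodes.
# The connections from these nodes to material inputs are not shown.
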Material /Game/Blueprints/Adventures/SpawnTeleporter/M_SpawnFlag_Pole
graph {
 properties {pixelDepthOffsetMode=PDOM_Legacy, usedShadingModels=Default Lit}
 nodes [
  Constant3Vector() {constant=<rgba 0.235 0.235 0.235 0>}
  Constant() {r=0.35}
  Constant() {r=1}
  Constant() {r=0.6}
  VectorParameter() {defaultValue=<rgba 0.25 0.25 0.25 0>}
 ]
}
